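# UI material for the TidalCraneHoagie altitude meter, an alpha-composited bar with a curved fill driven by a scalar Progress parameter.
# An MSDF bar texture becomes an SDF, thresholded with anti-aliasing into bar and outline masks.
# Progress remaps into an offset radial arc's fill ratio, while an offset, scaled circle SDF forms a GreyLine mask.
# These named-reroute masks blend greys with orange-red and set bar and outline opacity.
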Material /TidalCraneHoagie/UI/Materials/M_UI_TidalCraneHoagie_AltitudeMeter
graph {
 properties {bAutomaticallySetUsageInEditor=false, blendMode=BLEND_AlphaComposite, materialDomain=MD_UI, usedShadingModels=Default Lit}
 nodes [
  n1 = LinearInterpolate(
   A=Constant3Vector() {constant=<rgba 0.4793 0.4793 0.4793 1>}
   B=Constant3Vector() {constant=<rgba 0.2747 0.2747 0.2747 1>}
   Alpha=NamedRerouteUsage()
  )
  n2 = Multiply(A=Multiply(A=Constant3Vector() {constant=<rgba 1 1 1 1>}) {constB=0.3}, B=@n1)
  n3 = NamedRerouteUsage()
  n4 = NamedRerouteUsage()
  n5 = Multiply(A=@n4, B=NamedRerouteUsage()) {constB=0.5}
  n6 = Constant() {desc=BackgroundOpacity}
  n7 = NamedRerouteUsage()
  Multiply(
   A=LinearInterpolate(A=Constant3Vector() {constant=<rgba 0.8796 0.1195 0.0144 1>}, B=LinearInterpolate(A=@n1, B=@n2, Alpha=@n3), Alpha=NamedRerouteUsage())
   B=Max(A=LinearInterpolate(A=@n4, B=@n5, Alpha=@n6) {constAlpha=0.9}, B=Multiply(A=@n7) {constB=0.4, desc=OutlineOpacity})
  )
  NamedRerouteDeclaration(None=TextureCoordinate()) {name=UV, nodeColor=<rgba 0 0 0 1>}
  n8 = NamedRerouteUsage()
  n9 = Multiply(A=Constant2Vector() {g=1, r=0.421}, B=Constant() {desc=Bar Scale, r=1}) {constB=0.9}
  n10 = MaterialFunctionCall!(MSDF RGB=TextureSample(UVs=MaterialFunctionCall(UVs=@n8, Texture Scale=@n9) {materialFunction=&ScaleUVsByCenter}.UVs) {texture=/TidalCraneHoagie/UI/Assets/T_UI_TidalCraneHoagie_Altitude_Bar_MSDF}.RGB) {materialFunction=/FortUILibrary/MaterialFunctions/Shapes/Modifiers/MF_UI_MSDFtoSDF}
  NamedRerouteDeclaration(None=MaterialFunctionCall(DistanceField=@n10.SDF, Threshold=Constant() {r=0.3}) {materialFunction=&MF_UI_AntiAliasing}.Outer) {name=Bar, nodeColor=<rgba 0 0 0 1>}
  n11 = Add(A=NamedRerouteUsage(), B=Constant2Vector() {r=-0.8})
  n12 = LinearInterpolate(Alpha=ScalarParameter() {defaultValue=0.4, parameterName=Progress}) {constA=0.322, constB=0.178}
  NamedRerouteDeclaration(None=SmoothStep(Value=MaterialFunctionCall!(UV=@n11, "Arc Start Angle (0-1)", "Arc End Angle (0-1)", "Fill Ratio (0-1)"=@n12, Fill Clockwise) {materialFunction=/FortUILibrary/MaterialFunctions/Basics/MF_UI_RadialArc}.Radial Fill) {constMax=0.0005}) {name=Progress, nodeColor=<rgba 0 0 0 1>}
  n13 = MaterialFunctionCall(DistanceField=@n10.SDF, Threshold=Constant() {r=0.24}) {materialFunction=&MF_UI_AntiAliasing}
  n14 = MaterialFunctionCall(DistanceField=@n10.SDF, Threshold=Constant() {r=0.27}) {materialFunction=&MF_UI_AntiAliasing}
  NamedRerouteDeclaration(None=Saturate(None=Subtract(A=@n13.Outer, B=@n14.Outer))) {name=Outline, nodeColor=<rgba 0 0 0 1>}
  n15 = MaterialFunctionCall(UVs=Add(A=NamedRerouteUsage(), B=Constant2Vector() {g=-0.005, r=-0.318}), Texture Scale=Constant2Vector() {g=1.22, r=0.7}) {materialFunction=&ScaleUVsByCenter}
  NamedRerouteDeclaration(None=MaterialFunctionCall(
    DistanceField=MaterialFunctionCall(UVs=@n15.UVs) {materialFunction=&MF_UI_Circle}.SDF
    Threshold=Constant() {r=0.5}
   ) {materialFunction=&MF_UI_AntiAliasing}.Inner) {name=GreyLine, nodeColor=<rgba 0 0 0 1>}
 ]
}
