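# UI material for TidalCraneHoagie that draws a bracket gauge with a moving meter mark, middle line, skewed lines and a glowing orange pentagon.
# Progress blends MinValue to MaxValue to rotate the meter box around an off-screen pivot, sweeping it along a shallow arc.
# Bracket and pentagon come from MSDF textures, lines from procedural boxes, with skewed pairs multiplied into slivers.
# Named-reroute masks get individual opacities, merge by Max and multiply a white-to-orange color for alpha-composite blending.
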Material /TidalCraneHoagie/UI/Materials/M_UI_TidalCraneHoagie_Brackets
graph {
 properties {bAutomaticallySetUsageInEditor=false, blendMode=BLEND_AlphaComposite, materialDomain=MD_UI, usedShadingModels=Default Lit}
 nodes [
  n1 = ScalarParameter() {parameterName=MinValue}
  n2 = ScalarParameter() {defaultValue=-0.042, parameterName=MaxValue}
  n3 = Clamp(None=ScalarParameter() {defaultValue=0.2, parameterName=Progress}) {minDefault=-1}
  n4 = MaterialFunctionCall(
   UVs=MaterialFunctionCall(
    UVs=NamedRerouteUsage()
    Rotation Center=Constant2Vector() {g=0.5, r=2}
    "Rotation Angle (0-1)"=LinearInterpolate(A=@n1, B=@n2, Alpha=@n3) {constA=0.038, constB=-0.038}
   ) {materialFunction=&CustomRotator}.Rotated Values
   Size=Constant2Vector() {g=0.012, r=0.052}
  ) {materialFunction=&MF_UI_Box}
  n5 = Constant() {r=1.1}
  MaterialFunctionCall(DistanceField=OneMinus(None=@n4.SDF), Threshold=@n5) {materialFunction=&MF_UI_AntiAliasing}
  n6 = MaterialFunctionCall(UVs=Add(A=NamedRerouteUsage(), B=Constant2Vector() {r=0.014}) {constB=0}, Size=Constant2Vector() {g=0.012, r=0.08}) {materialFunction=&MF_UI_Box}
  MaterialFunctionCall(DistanceField=OneMinus(None=@n6.SDF), Threshold=@n5) {materialFunction=&MF_UI_AntiAliasing}
  n7 = Constant3Vector() {desc=Glow Color, constant=<rgba 1 0.1119 0 1>}
  n8 = Constant3Vector() {desc=Pentagon Color, constant=<rgba 1 0.5029 0 1>}
  n9 = NamedRerouteUsage()
  n10 = NamedRerouteUsage()
  n11 = NamedRerouteUsage()
  n12 = Constant() {desc=Bracket Opacity, r=0.003}
  n13 = NamedRerouteUsage()
  n14 = Constant() {desc=Meter Opacity, r=0.3}
  n15 = Max(
   A=Max(A=Multiply(A=@n11, B=@n12), B=Multiply(A=@n13, B=@n14) {constB=0.3})
   B=Multiply(A=NamedRerouteUsage(), B=Constant() {desc=Middle Line Opacity, r=0.01})
  )
  n16 = Multiply(A=NamedRerouteUsage(), B=Constant() {desc=Skew Lines Opcity, r=0.006}) {constB=0.1}
  n17 = NamedRerouteUsage()
  n18 = Multiply(A=NamedRerouteUsage(), B=Constant() {desc=Pentagon Outline, r=0.75})
  Multiply(
   A=LinearInterpolate(A=Constant3Vector() {constant=<rgba 1 1 1 0>}, B=LinearInterpolate(A=@n7, B=@n8, Alpha=@n9), Alpha=Ceil(None=@n10))
   B=Max(A=Max(A=@n15, B=@n16), B=Max(A=@n17, B=@n18))
  )
  n19 = NamedRerouteUsage()
  n20 = Constant2Vector() {r=-0.08}
  n21 = Constant2Vector() {g=1, r=0.2}
  n22 = Constant() {desc=Bar Scale, r=1}
  n23 = TextureSample(UVs=MaterialFunctionCall(UVs=Add(A=@n19, B=@n20), Texture Scale=Multiply(A=@n21, B=@n22) {constB=0.9}) {materialFunction=&ScaleUVsByCenter}.UVs) {texture=/TidalCraneHoagie/UI/Assets/T_UI_TidalCraneHoagie_InnerBracket_MSDF}
  NamedRerouteDeclaration(None=MaterialFunctionCall(
    DistanceField=MaterialFunctionCall!(MSDF RGB=@n23.RGB) {materialFunction=/FortUILibrary/MaterialFunctions/Shapes/Modifiers/MF_UI_MSDFtoSDF}.SDF
    Threshold=Constant() {r=0.28}
   ) {materialFunction=&MF_UI_AntiAliasing}.Outer) {name=Bracket, nodeColor=<rgba 0 0 0 1>}
  NamedRerouteDeclaration(None=TextureCoordinate()) {name=UV, nodeColor=<rgba 0 0 0 1>}
  NamedRerouteDeclaration(None=@n4.Fill) {name=Meter, nodeColor=<rgba 0 0 0 1>}
  NamedRerouteDeclaration(None=@n6.Fill) {name=MiddleLine, nodeColor=<rgba 0 0 0 1>}
  n24 = NamedRerouteUsage()
  n25 = Constant2Vector() {g=-0.056, r=0.096}
  n26 = Constant2Vector() {g=0.2, r=0.085}
  n27 = MaterialFunctionCall(
   UVs=MaterialFunctionCall!(UV=Add(A=@n24, B=@n25), Skew Angle=Constant2Vector() {g=-45.6643}) {materialFunction=/FortUILibrary/MaterialFunctions/Basics/MF_UI_Skew}.Skewed UVs
   Size=@n26
  ) {materialFunction=&MF_UI_Box}
  n28 = Constant2Vector() {g=0.144, r=0.112}
  n29 = MaterialFunctionCall(
   UVs=MaterialFunctionCall(UV=Add(A=@n24, B=@n28), Skew Angle=Constant2Vector() {g=-36.9227}) {materialFunction=&MF_UI_Skew}.Skewed UVs
   Size=@n26
  ) {materialFunction=&MF_UI_Box}
  n30 = NamedRerouteUsage()
  n31 = Constant2Vector() {g=-0.133, r=0.096}
  n32 = Constant2Vector() {g=0.2, r=0.085}
  n33 = MaterialFunctionCall(
   UVs=MaterialFunctionCall(UV=Add(A=@n30, B=@n31), Skew Angle=Constant2Vector() {g=36.9227}) {materialFunction=&MF_UI_Skew}.Skewed UVs
   Size=@n32
  ) {materialFunction=&MF_UI_Box}
  n34 = Constant2Vector() {g=0.038, r=0.112}
  n35 = MaterialFunctionCall(
   UVs=MaterialFunctionCall(UV=Add(A=@n30, B=@n34), Skew Angle=Constant2Vector() {g=45.6643}) {materialFunction=&MF_UI_Skew}.Skewed UVs
   Size=@n32
  ) {materialFunction=&MF_UI_Box}
  NamedRerouteDeclaration(None=Add(A=Multiply(A=@n27.Fill, B=@n29.Fill), B=Multiply(A=@n33.Fill, B=@n35.Fill))) {name=SkewLines, nodeColor=<rgba 0 0 0 1>}
  n36 = Add(A=NamedRerouteUsage(), B=Constant2Vector() {r=0.08})
  n37 = Constant() {r=0.05}
  n38 = MaterialFunctionCall(MSDF RGB=TextureSample(UVs=MaterialFunctionCall(UVs=@n36, Texture Scale=@n37) {materialFunction=&ScaleUVsByCenter}.UVs) {texture=/TidalCraneHoagie/UI/Assets/T_UI_TidalCraneHoagie_Pentagon_MSDF}.RGB) {materialFunction=&MF_UI_MSDFtoSDF}
  NamedRerouteDeclaration(None=MaterialFunctionCall(DistanceField=@n38.SDF, Threshold=Constant() {r=0.6}) {materialFunction=&MF_UI_AntiAliasing}.Outer) {name=Pentagon, nodeColor=<rgba 0 0 0 1>}
  NamedRerouteDeclaration(None=MaterialFunctionCall(DistanceField=@n38.SDF, Threshold=Constant() {r=0.5}, Smoothing=Constant() {r=0.9}) {materialFunction=&MF_UI_AntiAliasing}.Outer) {name=PentagonOutline, nodeColor=<rgba 0 0 0 1>}
 ]
}
