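# Translucent UI material for the TidalCraneHoagie center graphic, drawing a faint ring and outlined square shapes procedurally.
# The ring is an anti-aliased circle SDF outline.
# The squares are anti-aliased box SDF outlines on rotated, offset UVs from the FortUILibrary reticle function.
# RingOpacity and SquaresOpacity scale two mask inputs, which a Max node combines so overlaps do not add up.
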Material /TidalCraneHoagie/UI/Materials/M_UI_TidalCraneHoagie_Center
graph {
 properties {bAutomaticallySetUsageInEditor=false, blendMode=BLEND_Translucent, materialDomain=MD_UI, usedShadingModels=Default Lit}
 nodes [
  Constant() {r=1}
  Max(
   A=Multiply(A=NamedRerouteUsage(), B=ScalarParameter() {defaultValue=0.1, parameterName=RingOpacity})
   B=Multiply(A=NamedRerouteUsage(), B=ScalarParameter() {defaultValue=0.05, parameterName=SquaresOpacity})
  )
  n1 = MaterialFunctionCall(Size=Constant() {r=0.7}, OutlineThickness=Constant() {r=0.001}) {materialFunction=&MF_UI_Circle}
  NamedRerouteDeclaration(None=MaterialFunctionCall(DistanceField=Add(A=@n1.SDFOutline) {constB=0.08}, Threshold=Constant() {r=0.1}) {materialFunction=&MF_UI_AntiAliasing}.Inner) {name=Ring, nodeColor=<rgba 0 0 0 1>}
  n2 = MaterialFunctionCall!(
   Reticle UVs
   Reticle Sides
   Reticle Scale
   Reticle Rotation=Constant() {r=0.125}
   Prong Scale
   Prong Rotation
   Spread Offset=Constant() {r=-0.07}
   Positional Offset
   "Spread Input Range [Min,Max]"
   "Spread Remapped Range [Min,Max]"
  ) {materialFunction=/FortUILibrary/Systems/Reticles/MaterialFunctions/MF_UI_Reticle_UVs}
  n3 = Constant() {r=0.15}
  n4 = Constant() {r=-1}
  n5 = Constant() {r=0.02}
  NamedRerouteDeclaration(None=MaterialFunctionCall(
    DistanceField=MaterialFunctionCall(UVs=@n2.UVs, Size=@n3, CornerRoundness=@n4, OutlineThickness=@n5) {materialFunction=&MF_UI_Box}.SDFOutline
    Threshold=Constant() {r=0.02}
   ) {materialFunction=&MF_UI_AntiAliasing}.Inner) {name=Squares, nodeColor=<rgba 0 0 0 1>}
 ]
}
