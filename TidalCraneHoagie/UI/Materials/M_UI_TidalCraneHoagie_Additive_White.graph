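# UI-domain material with additive blending, named as the white additive variant for TidalCraneHoagie UI.
# The graph holds a single vector parameter, RGBA, defaulting to opaque white, which material instances can override.
# No textures, math or links from the parameter to material outputs appear in the graph.
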Material /TidalCraneHoagie/UI/Materials/M_UI_TidalCraneHoagie_Additive_White
graph {
 properties {blendMode=BLEND_Additive, materialDomain=MD_UI, pixelDepthOffsetMode=PDOM_Legacy, usedShadingModels=Default Lit}
 nodes [VectorParameter() {parameterName=RGBA, defaultValue=<rgba 1 1 1 1>}]
}
