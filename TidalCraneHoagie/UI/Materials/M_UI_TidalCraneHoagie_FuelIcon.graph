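# UI-domain translucent material for the TidalCraneHoagie low fuel icon, with parameters for color, opacity and scale.
# The icon texture is sampled with UVs scaled around their center, then read as a distance field by MF_UI_AntiAliasing with fixed threshold and smoothing.
# That function's Outer output, multiplied by opacity, drives a lerp toward the icon color.
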Material /TidalCraneHoagie/UI/Materials/M_UI_TidalCraneHoagie_FuelIcon
graph {
 properties {bAutomaticallySetUsageInEditor=false, blendMode=BLEND_Translucent, materialDomain=MD_UI, usedShadingModels=Default Lit}
 nodes [
  n1 = TextureSample(UVs=MaterialFunctionCall(UVs=TextureCoordinate(), Texture Scale=VectorParameter() {parameterName=Fuel Icon - Scale, sortPriority=15, defaultValue=<rgba 1 1 0 1>}.RGB) {materialFunction=&ScaleUVsByCenter}.UVs) {texture=/TidalCraneHoagie/UI/Assets/T_UI_TidalCraneHoagie_LowFuelIcon}
  n2 = Constant() {r=0.2}
  n3 = Constant() {r=0.1}
  LinearInterpolate(
   B=VectorParameter() {group=Fuel Icon, parameterName=Fuel Icon - Color, sortPriority=20, defaultValue=<rgba 1 1 1 0>}.RGB
   Alpha=Multiply(
    A=MaterialFunctionCall(DistanceField=@n1.RGB, Threshold=@n2, Smoothing=@n3) {materialFunction=&MF_UI_AntiAliasing}.Outer
    B=ScalarParameter() {defaultValue=1, group=Fuel Icon, parameterName=Fuel Icon - Opacity, sortPriority=10}
   )
  )
 ]
}
