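# Translucent UI material that, by its name, provides decoration for the fuel icon in the TidalCraneHoagie UI.
# It builds a procedural shape from two MF_UI_BoxAngle calls on the same UVs, each cutting at its own angle.
# A Max node merges the two masks, and a Multiply dims the result to a faint level, with no texture sampling.
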Material /TidalCraneHoagie/UI/Materials/M_UI_TidalCraneHoagie_FuelIconDecor
graph {
 properties {bAutomaticallySetUsageInEditor=false, blendMode=BLEND_Translucent, materialDomain=MD_UI, usedShadingModels=Default Lit}
 nodes [
  Constant() {r=1}
  n1 = TextureCoordinate()
  n2 = Constant() {r=70}
  n3 = Constant() {r=2}
  n4 = Constant() {r=15}
  n5 = Constant() {r=1}
  Multiply(A=Max(
    A=MaterialFunctionCall!(UV=@n1, Corner, Angle=@n2, Flip, Falloff) {materialFunction=/FortUILibrary/MaterialFunctions/Shapes/MF_UI_BoxAngle}.Result
    B=MaterialFunctionCall(UV=@n1, Corner=@n3, Angle=@n4, Flip=@n5) {materialFunction=&MF_UI_BoxAngle}.Result
   )) {constB=0.24}
 ]
}
