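# UI material for the TidalCraneHoagie fuel meter: a curved bar with brackets, outline and outer ring.
# MSDF textures and circle SDFs form the shapes; a radial arc driven by Progress masks the fill.
# The bar uses a three-color gradient, with low-fuel tinting controlled by threshold and ShowLowFuel parameters.
# The Gradient static switch swaps this for a glow-only layer, built with stepped, time-animated scratch and noise masks.
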Material /TidalCraneHoagie/UI/Materials/M_UI_TidalCraneHoagie_FuelMeter
graph {
 properties {bAutomaticallySetUsageInEditor=false, blendMode=BLEND_AlphaComposite, materialDomain=MD_UI, usedShadingModels=Default Lit}
 nodes [
  NamedRerouteDeclaration(None=TextureCoordinate()) {name=UV, nodeColor=<rgba 0 0 0 1>}
  n1 = NamedRerouteUsage()
  n2 = Multiply(A=Constant2Vector() {g=1, r=0.3}, B=Constant() {desc=Bar Scale, r=1}) {constB=0.9}
  n3 = MaterialFunctionCall!(MSDF RGB=TextureSample(UVs=MaterialFunctionCall(UVs=@n1, Texture Scale=@n2) {materialFunction=&ScaleUVsByCenter}.UVs) {texture=/TidalCraneHoagie/UI/Assets/T_UI_TidalCraneHoagie_Fuel_Bar_MSDF}.RGB) {materialFunction=/FortUILibrary/MaterialFunctions/Shapes/Modifiers/MF_UI_MSDFtoSDF}
  NamedRerouteDeclaration(None=MaterialFunctionCall(DistanceField=@n3.SDF, Threshold=Constant() {r=0.39}) {materialFunction=&MF_UI_AntiAliasing}.Outer) {name=Bar, nodeColor=<rgba 0 0 0 1>}
  n4 = ScalarParameter() {defaultValue=0.5, parameterName=Progress}
  n5 = MaterialFunctionCall!(
   UV=Add(A=NamedRerouteUsage(), B=Constant2Vector() {r=-0.86})
   "Arc Start Angle (0-1)"=Constant() {r=0.6783}
   "Arc End Angle (0-1)"=Constant() {r=0.1782}
   "Fill Ratio (0-1)"=OneMinus(None=Saturate(None=@n4))
   Fill Clockwise
  ) {materialFunction=/FortUILibrary/MaterialFunctions/Basics/MF_UI_RadialArc}
  n6 = SmoothStep(Value=@n5.Radial Fill) {constMax=0.0005}
  NamedRerouteDeclaration(None=Multiply(A=OneMinus(None=@n5.Arc Mask), B=@n6)) {name=Progress_Gradient, nodeColor=<rgba 0 0 0 1>}
  n7 = OneMinus(None=ScalarParameter() {defaultValue=0.25, desc=Low Fuel Percent, parameterName=Low Fuel Threshold})
  NamedRerouteDeclaration(None=Saturate(None=MaterialFunctionCall(Input=@n5.Arc Mask, Input Low=@n7) {materialFunction=&RemapValueRange}.Result)) {name=LowFuel_Gradient, nodeColor=<rgba 0 0 0 1>}
  n8 = MaterialFunctionCall(UVs=NamedRerouteUsage(), Texture Scale=Constant() {r=0.25}) {materialFunction=&ScaleUVsByCenter}
  n9 = Frac(None=Time())
  n10 = Multiply(A=@n9) {constB=0.5}
  n11 = Constant() {r=12}
  n12 = StaticBool()
  n13 = TextureSample(UVs=Add(
    A=MaterialFunctionCall(UV=@n8.UVs) {materialFunction=&MF_UI_UVScaling}.Scaled To Fill
    B=MaterialFunctionCall!("Param (0-1)"=@n10, Tiles=@n11, Apply Offset=@n12) {materialFunction=/Game/Environments/Papaya/Materials/MaterialFunctions/MF_Papaya_Quantize_Scalar}.Result
   )) {samplerType=SAMPLERTYPE_LinearGrayscale, texture=/TidalCraneHoagie/UI/Assets/T_UI_TidalCraneHoagie_Scratches}
  n14 = Multiply(A=@n9) {constB=2}
  n15 = Constant() {r=5}
  n16 = TextureSample(UVs=Add(
    A=MaterialFunctionCall(UV=@n8.UVs) {materialFunction=&MF_UI_UVScaling}.Scaled to Fit
    B=MaterialFunctionCall("Param (0-1)"=@n14, Tiles=@n15) {materialFunction=&MF_Papaya_Quantize_Scalar}.Result
   )) {samplerType=SAMPLERTYPE_Normal, texture=/Game/Global/Textures/Default/Noise/64x64Noise}
  NamedRerouteDeclaration(None=Add(A=Power(Base=@n13.R), B=Multiply(A=@n16.R) {constB=0.1})) {name=I_Glow_Scratch_Mask, nodeColor=<rgba 0 0 0 1>}
  NamedRerouteDeclaration(None=@n6) {name=Progress, nodeColor=<rgba 0 0 0 1>}
  n17 = MaterialFunctionCall!(Repeat Count=Constant2Vector() {g=2, r=1}, texcoord=NamedRerouteUsage()) {materialFunction=/Game/Environments/Papaya/Materials/MaterialFunctions/MF_UVMirror}
  n18 = Constant2Vector() {g=0.3, r=-0.05}
  n19 = TextureSample(UVs=MaterialFunctionCall(UVs=Add(A=@n17.Result, B=@n18) {constB=0}, Texture Scale=Constant2Vector() {g=0.32, r=0.16}) {materialFunction=&ScaleUVsByCenter}.UVs) {samplerSource=SSM_Clamp_WorldGroupSettings, texture=/TidalCraneHoagie/UI/Assets/T_UI_TidalCraneHoagie_Fuel_Bracket_MSDF}
  NamedRerouteDeclaration(None=MaterialFunctionCall(
    DistanceField=MaterialFunctionCall(MSDF RGB=@n19.RGB) {materialFunction=&MF_UI_MSDFtoSDF}.SDF
    Threshold=Constant() {r=0.38}
   ) {materialFunction=&MF_UI_AntiAliasing}.Outer) {name=Brackets, nodeColor=<rgba 0 0 0 1>}
  n20 = Constant() {r=0.24}
  n21 = Constant() {r=0.27}
  NamedRerouteDeclaration(None=Subtract(
    A=MaterialFunctionCall(DistanceField=@n3.SDF, Threshold=@n20) {materialFunction=&MF_UI_AntiAliasing}.Outer
    B=MaterialFunctionCall(DistanceField=@n3.SDF, Threshold=@n21) {materialFunction=&MF_UI_AntiAliasing}.Outer
   )) {name=Bar Outline, nodeColor=<rgba 0 0 0 1>}
  n22 = NamedRerouteUsage()
  n23 = NamedRerouteUsage()
  n24 = Constant2Vector() {g=-0.005, r=-0.187}
  n25 = MaterialFunctionCall(UVs=MaterialFunctionCall(UVs=Add(A=@n23, B=@n24), Texture Scale=Constant2Vector() {g=1, r=0.43}) {materialFunction=&ScaleUVsByCenter}.UVs) {materialFunction=&MF_UI_Circle}
  n26 = Constant() {r=0.5}
  n27 = Multiply(A=Constant() {r=0.14}) {constB=0.5}
  n28 = NamedRerouteUsage()
  n29 = Constant2Vector() {r=-0.4}
  n30 = Constant() {r=0.8}
  n31 = MaterialFunctionCall(
   UVs=MaterialFunctionCall(UVs=Add(A=@n28, B=@n29), Texture Scale=Constant2Vector() {g=1.25, r=0.728}) {materialFunction=&ScaleUVsByCenter}.UVs
   Size=Constant() {r=0.82}
   OutlineThickness=@n30
  ) {materialFunction=&MF_UI_Circle}
  n32 = MaterialFunctionCall(DistanceField=@n31.SDF, Threshold=Add(A=@n27) {constB=0.64}, Smoothing=@n30) {materialFunction=&MF_UI_AntiAliasing}
  n33 = Divide(B=Multiply(A=OneMinus(None=@n32.Inner)) {constB=10})
  n34 = Power(Base=@n32.Inner) {constExponent=2.5}
  n35 = Add(A=@n27) {constB=0.29}
  n36 = Saturate(None=Divide(A=Subtract(A=@n35, B=@n31.SDFOutline)) {constB=0.01})
  n37 = OneMinus(None=MaterialFunctionCall(Alpha=@n31.SDF, Min=@n35, Max=Add(A=@n35) {constB=0.01}) {desc=Inner Shape, materialFunction=&SmoothStep}.Result)
  n38 = Multiply(
   A=Subtract(
    A=NamedRerouteUsage()
    B=MaterialFunctionCall(DistanceField=@n25.SDF, Threshold=@n26) {materialFunction=&MF_UI_AntiAliasing}.Inner
   )
   B=Multiply(A=Subtract(A=@n36, B=@n37), B=Multiply(A=@n33, B=@n34))
  )
  n39 = Multiply(A=Saturate(None=Multiply(A=@n38, B=@n22))) {constB=0.7}
  n40 = Multiply(A=@n39, B=NamedRerouteUsage())
  NamedRerouteDeclaration(None=Multiply(A=Add(A=@n39, B=@n40)) {constB=0.1}) {name=Glow, nodeColor=<rgba 0 0 0 1>}
  n41 = NamedRerouteUsage()
  n42 = Constant2Vector() {r=-0.354}
  n43 = MaterialFunctionCall(
   UVs=MaterialFunctionCall(UVs=Add(A=@n41, B=@n42), Texture Scale=Constant2Vector() {g=1.5, r=0.85}) {materialFunction=&ScaleUVsByCenter}.UVs
   Size=Constant() {r=0.992}
  ) {materialFunction=&MF_UI_Circle}
  n44 = Constant() {r=0.006}
  n45 = MaterialFunctionCall(
   UV=Add(A=NamedRerouteUsage(), B=Constant2Vector() {r=-0.21})
   "Arc Start Angle (0-1)"=Constant() {r=0.58}
   "Fill Ratio (0-1)"=Constant() {r=0.19}
  ) {materialFunction=&MF_UI_RadialArc}
  NamedRerouteDeclaration(None=Multiply(
    A=MaterialFunctionCall!(DistanceField=@n43.SDF, Threshold, Thickness=@n44, Alignment, Smoothing) {materialFunction=/FortUILibrary/MaterialFunctions/Shapes/Modifiers/MF_UI_ShapeOutline}.Outline
    B=SmoothStep(Value=@n45.Radial Fill) {constMax=0.0005}
   )) {name=OuterRing, nodeColor=<rgba 0 0 0 1>}
  NamedRerouteDeclaration(None=SmoothStep(Min=Subtract(A=@n7) {constB=0.0005}, Max=@n7, Value=@n5.Arc Mask) {constMax=0.75, constMin=0.7495}) {name=LowFuel, nodeColor=<rgba 0 0 0 1>}
  n46 = NamedRerouteUsage()
  n47 = NamedRerouteUsage()
  n48 = LinearInterpolate(B=Constant3Vector() {constant=<rgba 1 0.0666 0 1>}, Alpha=Add(A=Add(A=@n46, B=@n47), B=NamedRerouteUsage()))
  n49 = MaterialFunctionCall(
   A=VectorParameter() {group=Color, parameterName=Bar Color - Start, sortPriority=20, defaultValue=<rgba 1 0.0666 0 1>}.RGB
   B=VectorParameter() {group=Color, parameterName=Bar Color - Mid, sortPriority=40, defaultValue=<rgba 1 0.2086 0 1>}.RGB
   C=VectorParameter() {group=Color, parameterName=Bar Color - End, sortPriority=60, defaultValue=<rgba 1 0.7231 0 1>}.RGB
   Alpha=NamedRerouteUsage()
  ) {materialFunction=&3ColorBlend}
  n50 = VectorParameter() {parameterName=Color_LowFuel, defaultValue=<rgba 0.3385 0.0076 0 1>}
  n51 = ScalarParameter() {defaultValue=0.95, parameterName=ShowLowFuel}
  n52 = Multiply(A=NamedRerouteUsage(), B=@n51) {constB=0}
  n53 = NamedRerouteUsage()
  n54 = LinearInterpolate(
   A=Constant3Vector() {constant=<rgba 0.0182 0.0182 0.0182 1>}
   B=LinearInterpolate(
    A=LinearInterpolate(A=@n49.Result, B=@n50.RGB, Alpha=@n52)
    B=Constant3Vector() {constant=<rgba 0.5 0.0112 0 1>}
    Alpha=Multiply(A=@n53, B=@n51) {constB=0}
   )
   Alpha=NamedRerouteUsage()
  )
  n55 = NamedRerouteUsage()
  n56 = NamedRerouteUsage()
  n57 = Multiply(A=@n56, B=NamedRerouteUsage()) {constB=0.5}
  n58 = Multiply(A=Add(A=Multiply(A=@n57) {constB=0.1}, B=Multiply(A=@n56) {constB=0.5, desc=Progress Range}))
  n59 = Multiply(A=NamedRerouteUsage()) {constB=0.4}
  n60 = Max(A=Max(A=Max(A=@n58, B=@n59), B=NamedRerouteUsage()), B=Multiply(A=NamedRerouteUsage()) {constB=0.2})
  StaticSwitchParameter(
   True=LinearInterpolate(A=Constant3Vector(), B=Constant3Vector() {constant=<rgba 2 0.2297 0 1>}, Alpha=NamedRerouteUsage()) {desc=Progress Bar Glow}
   False=Multiply(A=LinearInterpolate(A=@n48, B=@n54, Alpha=@n55) {desc=Frame / Outline}, B=@n60)
  ) {group=Shape, parameterName=Gradient, sortPriority=100}
  StaticSwitchParameter(True=Multiply(A=NamedRerouteUsage(), B=Constant() {desc=GlowOpacity, r=2}) {constB=0.8}, False=@n60) {group=Shape, parameterName=Gradient, sortPriority=100}
 ]
}
