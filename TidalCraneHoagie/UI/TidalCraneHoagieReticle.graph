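# Widget Blueprint reticle HUD for the TidalCrane Hoagie vehicle, showing speed, fuel, boost, horizon brackets and warnings.
# Tick feeds smoothed pitch/roll and boost charge into dynamic materials and toggles the critical-health flash animation.
# Parent-class events drive speed and fuel meters and low/empty fuel alerts; a bound delegate recolors boost state.
# Async gameplay-tag waits toggle the disabled warning; Construct plays the intro and adjusts touch/compact layout.
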WidgetBlueprint /TidalCraneHoagie/UI/TidalCraneHoagieReticle
animations [
 Critical {
  binding_count=1
  length_seconds=0.4803
  object=&self:Critical
  sequence {bindings [CommonBorder_0 {
     id=#1
     possessed_class=CommonBorder
     tracks [{
       type=MovieSceneColorTrack
       display_name=Content Color and Opacity
       sections [{
         type=MovieSceneColorSection
         end_frame=9
         start_frame=0
         channels [
          Color.R {type=MovieSceneScriptingFloatChannel, keys [[0,0.8229],[9.6057,0.5365]]}
          Color.G {type=MovieSceneScriptingFloatChannel, keys [[0,0.0312],[9.6057,0.0203]]}
          Color.B {type=MovieSceneScriptingFloatChannel, const [0,2]}
          Color.A {type=MovieSceneScriptingFloatChannel, keys [[0,0.9],[9.6057,0.5]]}
         ]
        }]
      }]
    }]}
 }
 Intro {
  binding_count=9
  length_seconds=1.1421
  object=&self:Intro
  sequence {bindings [
    Bracket_Left {
     id=#2
     possessed_class=Image
     tracks [
      {
       type=MovieScene2DTransformTrack
       sections [{
         type=MovieScene2DTransformSection
         channels [
          Translation.X {type=MovieSceneScriptingFloatChannel, keys [[7,50],[15,0]]}
          Translation.Y {type=MovieSceneScriptingFloatChannel}
          Angle {type=MovieSceneScriptingFloatChannel}
          Scale.X {type=MovieSceneScriptingFloatChannel}
          Scale.Y {type=MovieSceneScriptingFloatChannel}
          Shear.X {type=MovieSceneScriptingFloatChannel}
          Shear.Y {type=MovieSceneScriptingFloatChannel}
         ]
        }]
      }
      {
       type=MovieSceneFloatTrack
       display_name=Render Opacity
       sections [{
         type=MovieSceneFloatSection
         channels [None {type=MovieSceneScriptingFloatChannel, keys [[4.9287,0],[5.6087,1],[6.42,0],[7.179,1],[14.9163,1]]}]
        }]
      }
     ]
    }
    Image_Speedometer_Glow {
     id=#3
     possessed_class=Image
     tracks [
      {
       type=MovieSceneFloatTrack
       display_name=Render Opacity
       sections [{
         type=MovieSceneFloatSection
         channels [None {type=MovieSceneScriptingFloatChannel, keys [[4.9287,0],[5.6087,1],[6.42,0],[7.179,1],[14.9163,1]]}]
        }]
      }
      {
       type=MovieScene2DTransformTrack
       sections [{
         type=MovieScene2DTransformSection
         channels [
          Translation.X {type=MovieSceneScriptingFloatChannel, keys [[4.9287,50],[14.9163,0]]}
          Translation.Y {type=MovieSceneScriptingFloatChannel}
         ]
        }]
      }
     ]
    }
    VerticalBox_Fuel {
     id=#4
     possessed_class=VerticalBox
     tracks [
      {
       type=MovieSceneFloatTrack
       display_name=Render Opacity
       sections [{type=MovieSceneFloatSection, end_frame=15, start_frame=0, channels [None {type=MovieSceneScriptingFloatChannel, keys [[0,0],[12,0],[15,1]]}]}]
      }
      {
       type=MovieScene2DTransformTrack
       sections [{
         type=MovieScene2DTransformSection
         end_frame=15
         start_frame=0
         channels [
          Translation.X {type=MovieSceneScriptingFloatChannel, keys [[0,-6],[12,-6],[15,0]]}
          Translation.Y {type=MovieSceneScriptingFloatChannel, const [0,1]}
          Angle {type=MovieSceneScriptingFloatChannel, const [0,1]}
          Scale.X {type=MovieSceneScriptingFloatChannel, const [1,1]}
          Scale.Y {type=MovieSceneScriptingFloatChannel, const [1,1]}
          Shear.X {type=MovieSceneScriptingFloatChannel, const [0,1]}
          Shear.Y {type=MovieSceneScriptingFloatChannel, const [0,1]}
         ]
        }]
      }
     ]
    }
    Image_Speedometer {
     id=#5
     possessed_class=Image
     tracks [
      {
       type=MovieSceneFloatTrack
       display_name=Render Opacity
       sections [{
         type=MovieSceneFloatSection
         channels [None {type=MovieSceneScriptingFloatChannel, keys [[4.9287,0],[5.6087,1],[6.42,0],[7.179,1],[14.9163,1]]}]
        }]
      }
      {
       type=MovieScene2DTransformTrack
       sections [{
         type=MovieScene2DTransformSection
         channels [
          Translation.X {type=MovieSceneScriptingFloatChannel, keys [[4.9287,50],[14.9163,0]]}
          Translation.Y {type=MovieSceneScriptingFloatChannel}
         ]
        }]
      }
     ]
    }
    Image_FuelMeter_Glow {
     id=#6
     possessed_class=Image
     tracks [
      {
       type=MovieSceneFloatTrack
       display_name=Render Opacity
       sections [{
         type=MovieSceneFloatSection
         channels [None {type=MovieSceneScriptingFloatChannel, keys [[4.9287,0],[5.6087,1],[6.42,0],[7.179,1],[14.9163,1]]}]
        }]
      }
      {
       type=MovieScene2DTransformTrack
       sections [{
         type=MovieScene2DTransformSection
         channels [
          Translation.X {type=MovieSceneScriptingFloatChannel, keys [[5,-31.6228],[14.9163,0]]}
          Translation.Y {type=MovieSceneScriptingFloatChannel}
         ]
        }]
      }
     ]
    }
    Bracket_Right {
     id=#7
     possessed_class=Image
     tracks [
      {
       type=MovieScene2DTransformTrack
       sections [{
         type=MovieScene2DTransformSection
         channels [
          Translation.X {type=MovieSceneScriptingFloatChannel, keys [[7,-50],[14.9163,0]]}
          Translation.Y {type=MovieSceneScriptingFloatChannel}
          Angle {type=MovieSceneScriptingFloatChannel}
          Scale.X {type=MovieSceneScriptingFloatChannel}
          Scale.Y {type=MovieSceneScriptingFloatChannel}
          Shear.X {type=MovieSceneScriptingFloatChannel}
          Shear.Y {type=MovieSceneScriptingFloatChannel}
         ]
        }]
      }
      {
       type=MovieSceneFloatTrack
       display_name=Render Opacity
       sections [{
         type=MovieSceneFloatSection
         channels [None {type=MovieSceneScriptingFloatChannel, keys [[4.9287,0],[5.6087,1],[6.42,0],[7.179,1],[14.9163,1]]}]
        }]
      }
     ]
    }
    Image_BoostMeter {
     id=#8
     possessed_class=Image
     tracks [
      {
       type=MovieScene2DTransformTrack
       sections [{
         type=MovieScene2DTransformSection
         channels [
          Translation.X {type=MovieSceneScriptingFloatChannel, keys [[4.9287,-50],[14.9163,0]]}
          Translation.Y {type=MovieSceneScriptingFloatChannel}
          Angle {type=MovieSceneScriptingFloatChannel}
          Scale.X {type=MovieSceneScriptingFloatChannel}
          Scale.Y {type=MovieSceneScriptingFloatChannel}
          Shear.X {type=MovieSceneScriptingFloatChannel}
          Shear.Y {type=MovieSceneScriptingFloatChannel}
         ]
        }]
      }
      {
       type=MovieSceneFloatTrack
       display_name=Render Opacity
       sections [{
         type=MovieSceneFloatSection
         channels [None {type=MovieSceneScriptingFloatChannel, keys [[4.9287,0],[5.6087,1],[6.42,0],[7.179,1],[14.9163,1]]}]
        }]
      }
     ]
    }
    Image_Center {
     id=#9
     possessed_class=Image
     tracks [{
       type=MovieSceneFloatTrack
       display_name=Render Opacity
       sections [{
         type=MovieSceneFloatSection
         channels [None {
           type=MovieSceneScriptingFloatChannel
           keys [[0,1],[2.2503,0],[2.9303,1],[3.7417,0],[4.5007,1],[20.4573,1],[20.5913,0],[21.2713,1],[22.0827,0],[22.8417,1]]
          }]
        }]
      }]
    }
    Image_FuelMeter {
     id=#10
     possessed_class=Image
     tracks [
      {
       type=MovieScene2DTransformTrack
       sections [{
         type=MovieScene2DTransformSection
         channels [
          Translation.X {type=MovieSceneScriptingFloatChannel, keys [[5,-31.6228],[14.9163,0]]}
          Translation.Y {type=MovieSceneScriptingFloatChannel}
         ]
        }]
      }
      {
       type=MovieSceneFloatTrack
       display_name=Render Opacity
       sections [{
         type=MovieSceneFloatSection
         channels [None {type=MovieSceneScriptingFloatChannel, keys [[4.9287,0],[5.6087,1],[6.42,0],[7.179,1],[14.9163,1]]}]
        }]
      }
     ]
    }
   ]}
 }
 Intro_FuelInfo {
  binding_count=2
  length_seconds=0.75
  object=&self:Intro_FuelInfo
  sequence {bindings [
    Fuel_Low {
     id=#11
     possessed_class=Image
     tracks [{
       type=MovieSceneFloatTrack
       display_name=Render Opacity
       sections [{type=MovieSceneFloatSection, end_frame=15, start_frame=0, channels [None {type=MovieSceneScriptingFloatChannel, keys [[0,0],[13,0],[15,1]]}]}]
      }]
    }
    Overlay_Fuel {
     id=#12
     possessed_class=Overlay
     tracks [
      {
       type=MovieScene2DTransformTrack
       sections [{
         type=MovieScene2DTransformSection
         end_frame=15
         start_frame=0
         channels [
          Translation.X {type=MovieSceneScriptingFloatChannel, keys [[0,296],[13,296],[15,300]]}
          Translation.Y {type=MovieSceneScriptingFloatChannel, const [0,2]}
          Angle {type=MovieSceneScriptingFloatChannel, const [0,2]}
          Scale.X {type=MovieSceneScriptingFloatChannel, const [1,2]}
          Scale.Y {type=MovieSceneScriptingFloatChannel, const [1,2]}
          Shear.X {type=MovieSceneScriptingFloatChannel, const [0,2]}
          Shear.Y {type=MovieSceneScriptingFloatChannel, const [0,2]}
         ]
        }]
      }
      {
       type=MovieSceneFloatTrack
       display_name=Render Opacity
       sections [{type=MovieSceneFloatSection, end_frame=15, start_frame=0, channels [None {type=MovieSceneScriptingFloatChannel, keys [[0,0],[13,0],[15,1]]}]}]
      }
     ]
    }
   ]}
 }
]
graphs [
 EventGraph {nodes [
   n1 = HandleBoostStateChanged!() -> (OutputDelegate, then, State:EHoagieBoostState Enum=Unknown) {is=title K2Node_CustomEvent}
   > UpdateBoostState!(execute, self, State:EHoagieBoostState Enum=@n1.State) -> (then)
   n2 = Event Tick()
   > n3 = Is Valid(InputObject=Get OwningHoagieVehicle() -> (OwningHoagieVehicle:Fort Hoagie Vehicle Object Reference))
   SetRenderOpacity(execute=@n3.Is Not Valid, InOpacity=0.2)
   SetRenderOpacity(execute=@n3.Is Valid, InOpacity=1)
   > UpdatePitchRoll!(execute, self, dt:double=@n2.InDeltaTime) -> (then)
   > UpdateBoostValue()
   > UpdateCriticalHealth()
   n4 = Event Setup MID!() -> (OutputDelegate, then, bVehicleUsesFuelSystem:bool) {is=title K2Node_Event}
   > n5 = Sequence!(execute) -> (then_0, then_1, then_2:Exec, then_3:Exec)
   Bind Event to On Boost State Changed(execute=@n5.then_2, self:Fort Hoagie Vehicle Object Reference=Get OwningHoagieVehicle(), Delegate=@n1) {is=title K2Node_AddDelegate}
   n6 = Get OwningHoagieVehicle()
   > n7 = Branch(Condition=IsBoosting!(self:Fort Vehicle Interface Interface=@n6) -> (ReturnValue:bool))
   > UpdateBoostState(State=Started)
   n8 = Branch(
    execute=@n7.else
    Condition="float > float"(A=GetBoostCooldown!(self:Fort Hoagie Vehicle Object Reference=@n6) -> (ReturnValue:float), B=0) {is=title K2Node_CallFunction}
   )
   > UpdateBoostState(State=Finished)
   UpdateBoostState(execute=@n8.else, State=Ready)
   Set VehicleMaxAltitude(
    execute=@n5.then_1
    VehicleMaxAltitude:double=GetMaxAltitude!(self:Fort Hoagie Vehicle Object Reference=Get OwningHoagieVehicle()) -> (ReturnValue:float)
   ) -> (Output_Get:double)
   n9 = GetFirstLocalFortPlayerController!(bIncludeNonSpectators:bool=true, bIncludeSpectators:bool=false) -> (ReturnValue:Fort Player Controller Object Reference)
   Set MaxAltitudeWithoutFallDamage(
    execute=@n5.then_0
    MaxAltitudeWithoutFallDamage:float=GetMaxDistanceWithoutFallDamage!(self:Fort Pawn Object Reference=GetPlayerPawn!(self:Fort Player Controller Object Reference=@n9) -> (ReturnValue:Fort Player Pawn Object Reference)) -> (ReturnValue:float)
   ) -> (Output_Get:float)
   > n10 = EvaluateCurveTableRow!(
    execute:Exec
    CurveTable:Curve Table Object Reference=/TidalCraneHoagie/DataTables/TidalCraneHoagieGameData
    RowName:Name=Default.Vehicle.TidalCrane.Hoagie.Fuel.LowFuelPercent
    InXY:float=0
    ContextString:String
   ) -> (RowFound:Exec, RowNotFound:Exec, OutXY:float) {is=title K2Node_CallDataTableFunction}
   > n11 = GetDynamicMaterial(self=Get Image_FuelMeter() -> (Image_FuelMeter:Image Object Reference))
   > SetScalarParameterValue(self=@n11, ParameterName=Low Fuel Threshold, Value=@n10)
   SetVisibility(
    execute=@n5.then_3
    self=[
     Get CommonBorder_FuelMeterContainer() -> (CommonBorder_FuelMeterContainer:Common Border Object Reference)
     Get Image_Speedometer() -> (Image_Speedometer:Image Object Reference)
     Get Image_Speedometer_Glow() -> (Image_Speedometer_Glow:Image Object Reference)
    ]
    InVisibility=Select(Option 0=Collapsed, Option 1=SelfHitTestInvisible, Index=@n4.bVehicleUsesFuelSystem)
   )
   Event OnLowFuel() {is=title K2Node_Event}
   > UpdateFuelLow!(execute, self, In Low Fuel:bool=true) -> (then)
   Event Construct()
   > SetMobileLayout()
   > UpdateBoostValue()
   > n12 = GetDynamicMaterial(self=Get Bracket_Left() -> (Bracket_Left:Image Object Reference))
   > Set HorizonMIDLeft(HorizonMIDLeft:Material Instance Dynamic Object Reference=@n12) -> (Output_Get:Material Instance Dynamic Object Reference)
   > n13 = GetDynamicMaterial(self=Get Bracket_Right() -> (Bracket_Right:Image Object Reference))
   > Set HorizonMIDRight(HorizonMIDRight:Material Instance Dynamic Object Reference=@n13) -> (Output_Get:Material Instance Dynamic Object Reference)
   n14 = Get Intro()
   > StopAnimation(InAnimation=@n14)
   > PlayAnimation(InAnimation=@n14)
   > n15 = Branch(Condition=HasInfiniteFuel!(self:Fort Vehicle Interface Interface=Get OwningHoagieVehicle()) -> (ReturnValue:bool))
   > UpdateFuelEmpty!(execute, self, In No Fuel:bool=false) -> (then)
   > UpdateFuelLow(In Low Fuel=false)
   n16 = Get OwningHoagieVehicle()
   n17 = IsVehicleFuelEmpty!(self:Fort Vehicle Interface Interface=@n16) -> (ReturnValue:bool)
   UpdateFuelEmpty(execute=@n15.else, In No Fuel=@n17)
   > n18 = Branch(Condition=@n17)
   UpdateFuelLow(execute=@n18.else, In Low Fuel=IsVehicleFuelLow!(self:Fort Vehicle Interface Interface=@n16) -> (ReturnValue:bool))
   n19 = Event UpdateSpeedometer!() -> (OutputDelegate, then, NormalizedSpeed:float, RawSpeed:float) {is=title K2Node_Event}
   > n20 = GetDynamicMaterial(self=Get Image_Speedometer())
   > SetScalarParameterValue(self=@n20, ParameterName=Progress, Value=@n19.NormalizedSpeed)
   > n21 = GetDynamicMaterial(self=Get Image_Speedometer_Glow())
   > SetScalarParameterValue(self=@n21, ParameterName=Progress, Value=@n19.NormalizedSpeed)
   n22 = Event SetVehicleHasInfiniteFuel!() -> (OutputDelegate, then, bInfiniteFuel:bool) {is=title K2Node_Event}
   > Branch(Condition=@n22.bInfiniteFuel)
   > n23 = GetDynamicMaterial(self=Get Image_FuelMeter())
   > SetScalarParameterValue(self=@n23, ParameterName=Progress, Value=1)
   > n24 = GetDynamicMaterial(self=Get Image_FuelMeter_Glow() -> (Image_FuelMeter_Glow:Image Object Reference))
   > SetScalarParameterValue(self=@n24, ParameterName=Progress, Value=1)
   n25 = Event OnFuelValueChanged!() -> (OutputDelegate, then, FuelPercentage:float, bIsLowOnFuel:bool) {is=title K2Node_Event}
   > n26 = GetDynamicMaterial(self=Get Image_FuelMeter())
   > SetScalarParameterValue(self=@n26, ParameterName=Progress, Value=@n25.FuelPercentage)
   > n27 = GetDynamicMaterial(self=Get Image_FuelMeter_Glow())
   > SetScalarParameterValue(self=@n27, ParameterName=Progress, Value=@n25.FuelPercentage)
   n28 = Get OwningHoagieVehicle()
   n29 = IsVehicleFuelLow(self=@n28)
   n30 = IsVehicleFuelEmpty(self=@n28)
   > Branch(Condition=AND Boolean(A=NOT Boolean(A=@n30), B=NOT Boolean(A=@n29)))
   > SetVisibility(
    self=[Get Overlay_Fuel() -> (Overlay_Fuel:Overlay Object Reference), Get Fuel_Low() -> (Fuel_Low:Image Object Reference)]
    InVisibility=Collapsed
   )
   Event OnOutOfFuel() {is=title K2Node_Event}
   > UpdateFuelEmpty(In No Fuel=true)
   n31 = Event OnVehicleUpdated!() -> (OutputDelegate, then, NewVehicle:Fort Athena Vehicle Object Reference) {is=title K2Node_Event}
   n32 = Get DisabledTag() -> (DisabledTag:Gameplay Tag Structure)
   > n33 = WaitGameplayTagAddToActor!(execute:Exec, TargetActor:Actor Object Reference=@n31.NewVehicle, Tag:Gameplay Tag Structure=@n32, OnlyTriggerOnce:bool=false) -> (then:Exec, AsyncTaskProxy:Ability Async Wait Gameplay Tag Added Object Reference, Added:Exec) {is=title K2Node_AsyncAction}
   > n34 = WaitGameplayTagRemoveFromActor!(execute:Exec, TargetActor:Actor Object Reference=@n31.NewVehicle, Tag:Gameplay Tag Structure=@n32, OnlyTriggerOnce:bool=false) -> (then:Exec, AsyncTaskProxy:Ability Async Wait Gameplay Tag Removed Object Reference, Removed:Exec) {is=title K2Node_AsyncAction}
   OnVehicleEnabled!(execute=@n34.Removed, self, Enabled:bool=true) -> (then)
   OnVehicleEnabled(execute=@n33.Added, Enabled=false)
  ]}
 OnVehicleEnabled {nodes [
   n1 = OnVehicleEnabled!() -> (then, Enabled:bool=~) {is=title K2Node_FunctionEntry}
   > n2 = Branch(Condition=@n1)
   n3 = Get WBP_Vehicle_Warning() -> (WBP_Vehicle_Warning:WBP Vehicle Warning Object Reference)
   > RemoveWarning!(execute, self:WBP Vehicle Warning Object Reference=@n3, Type:ENUM_WarningType Enum=NewEnumerator2) -> (then)
   > SetVisibility(self=@n3, InVisibility=Collapsed)
   ShowWarning!(execute=@n2.else, self:WBP Vehicle Warning Object Reference=@n3, Type:ENUM_WarningType Enum=NewEnumerator2) -> (then)
   > SetVisibility(self=@n3, InVisibility=HitTestInvisible)
  ]}
 SetMobileLayout {nodes [
   SetMobileLayout() {is=title K2Node_FunctionEntry}
   > Branch(Condition=IsUsingCompactUI())
   > SetUserSpecifiedScale!(
    execute
    self:Scale Box Object Reference=Get ScaleBox_Fuel_Mobile() -> (ScaleBox_Fuel_Mobile:Scale Box Object Reference)
    InUserSpecifiedScale:float=1.5
   ) -> (then)
   n1 = IsTabletAspectRatio!() -> (ReturnValue:bool)
   n2 = GetOwningPlayer()
   > Branch(Condition=AND Boolean(
     A=IsUsingInputType!(InputType:ECommonInputType Enum=Touch, PlayerController:Player Controller Object Reference=@n2) -> (ReturnValue:bool)
     B=NOT Boolean(A=@n1)
    ))
   > SetRenderScale!(
    execute
    self:Widget Object Reference=[
     Get Overlay_LeftBracket() -> (Overlay_LeftBracket:Overlay Object Reference)
     Get Overlay_RightBracket() -> (Overlay_RightBracket:Overlay Object Reference)
    ]
    Scale:Vector 2D Structure=(X=0.85,Y=0.85)
   ) -> (then)
   > SetRenderTranslation(self=Get Overlay_LeftBracket(), Translation=(X=110,Y=0))
   > SetRenderTranslation(self=Get Overlay_RightBracket(), Translation=(X=-110,Y=0))
  ]}
 UpdateBoostState {nodes [
   n1 = UpdateBoostState!() -> (then, State:EHoagieBoostState Enum=Unknown) {is=title K2Node_FunctionEntry}
   > n2 = GetDynamicMaterial(self=Get Image_BoostMeter() -> (Image_BoostMeter:Image Object Reference))
   > SetVectorParameterValue(
    self=@n2
    ParameterName=Color_Overlay
    Value=Select!(
     Unknown:Linear Color Structure=(R=0.4793,G=0.4793,B=0.4793,A=1)
     Ready:Linear Color Structure=(R=0.4793,G=0.4793,B=0.4793,A=1)
     Started:Linear Color Structure=(R=1,G=0.5419,B=0,A=1)
     Finished:Linear Color Structure=(R=0.4508,G=0.0567,B=0.0423,A=0.713)
     Failed:Linear Color Structure=(R=0.4793,G=0.4793,B=0.4793,A=1)
     Index:EHoagieBoostState Enum=@n1
    ) -> (ReturnValue:Linear Color Structure)
   )
   n3 = Switch on EHoagieBoostState!(execute, Selection:EHoagieBoostState Enum=@n1) -> (Unknown:Exec, Ready:Exec, Started:Exec, Finished:Exec, Failed:Exec) {is=title K2Node_SwitchEnum}
   SetFillColorAndOpacity!(
    execute=@n3.Finished
    self:Progress Bar Object Reference=Get ProgressBar_Boost() -> (ProgressBar_Boost:Progress Bar Object Reference)
    InColor:Linear Color Structure=LinearColor_Yellow!() -> (ReturnValue:Linear Color Structure)
   ) -> (then)
   SetFillColorAndOpacity(
    execute=[@n3.Failed, @n3.Ready, @n3.Unknown]
    self=Get ProgressBar_Boost()
    InColor=LinearColor_White!() -> (ReturnValue:Linear Color Structure)
   )
   SetFillColorAndOpacity(execute=@n3.Started, self=Get ProgressBar_Boost(), InColor=LinearColor_Red!() -> (ReturnValue:Linear Color Structure))
  ]}
 UpdateBoostValue {nodes [
   UpdateBoostValue() {is=title K2Node_FunctionEntry}
   > n1 = Get!(execute:Exec) -> (then:Exec, else:Exec, OwningHoagieVehicle:Fort Hoagie Vehicle Object Reference)
   > n2 = GetDynamicMaterial(self=Get Image_BoostMeter())
   > SetScalarParameterValue(self=@n2, ParameterName=Progress, Value=GetRechargeableBoostPercent!(self:Fort Hoagie Vehicle Object Reference=@n1) -> (ReturnValue:float))
  ]}
 UpdateCriticalHealth {nodes [
   UpdateCriticalHealth() {is=title K2Node_FunctionEntry}
   > n1 = Branch(Condition=IsCriticalHealth!(self:Fort Athena Vehicle Object Reference=Get OwningHoagieVehicle()) -> (ReturnValue:bool))
   n2 = Get Critical() -> (Critical:Widget Animation Object Reference)
   > n3 = Branch(Condition=IsAnimationPlaying(InAnimation=@n2))
   PlayAnimation(execute=@n3.else, InAnimation=@n2)
   Branch(execute=@n1.else, Condition=IsAnimationPlaying(InAnimation=Get Critical()))
   > StopAnimation(InAnimation=@n2)
  ]}
 UpdateFuelEmpty {nodes [
   n1 = UpdateFuelEmpty!() -> (then, In No Fuel:bool=~) {is=title K2Node_FunctionEntry}
   > n2 = Branch(Condition=@n1)
   > SetVisibility(self=[Get Fuel_Low(), Get Overlay_Fuel()], InVisibility=SelfHitTestInvisible)
   > SetActiveWidgetIndex(self=Get Switcher_Fuel() -> (Switcher_Fuel:Common Widget Switcher Legacy Object Reference), Index=1)
   > PlayAnimationForward(InAnimation=Get Intro_FuelInfo() -> (Intro_FuelInfo:Widget Animation Object Reference))
   SetVisibility(execute=@n2.else, self=[Get Fuel_Low(), Get Overlay_Fuel()], InVisibility=Collapsed)
  ]}
 UpdateFuelLow {nodes [
   n1 = UpdateFuelLow!() -> (then, In Low Fuel:bool=~) {is=title K2Node_FunctionEntry}
   > n2 = Branch(Condition=@n1)
   > SetVisibility(self=[Get Fuel_Low(), Get Overlay_Fuel()], InVisibility=SelfHitTestInvisible)
   > SetActiveWidgetIndex(self=Get Switcher_Fuel(), Index=0)
   > PlayAnimationForward(InAnimation=Get Intro_FuelInfo())
   SetVisibility(execute=@n2.else, self=[Get Fuel_Low(), Get Overlay_Fuel()], InVisibility=Collapsed)
  ]}
 UpdatePitchRoll {nodes [
   n1 = UpdatePitchRoll!() -> (then, dt:double=~) {is=title K2Node_FunctionEntry}
   n2 = BreakRotator(InRot=Get Actor Rotation!(self:Actor Object Reference=Get OwningHoagieVehicle()) -> (ReturnValue:Rotator) {is=title K2Node_CallFunction})
   n3 = float / float(A=@n2.Pitch, B=-5) {is=title K2Node_CallFunction}
   > n4 = Set CurrentPitch(CurrentPitch:double=FInterpTo(Current=Get CurrentPitch() -> (CurrentPitch:double), Target="Clamp (Float)"(Value=@n3, Min=-1, Max=1), DeltaTime=@n1, InterpSpeed=0.75)) -> (Output_Get:double)
   > SetScalarParameterValue(self=Get HorizonMIDLeft() -> (HorizonMIDLeft:Material Instance Dynamic Object Reference), ParameterName=Pitch, Value=@n4)
   n5 = float / float(A=@n2, B=-25)
   > n6 = Set CurrentRoll(CurrentRoll:double=FInterpTo(Current=Get CurrentRoll() -> (CurrentRoll:double), Target="Clamp (Float)"(Value=@n5, Min=-1, Max=1), DeltaTime=@n1, InterpSpeed=0.75)) -> (Output_Get:double)
   > SetScalarParameterValue(self=Get HorizonMIDLeft(), ParameterName=Progress, Value=@n6)
   > SetScalarParameterValue(self=Get HorizonMIDRight() -> (HorizonMIDRight:Material Instance Dynamic Object Reference), ParameterName=Progress, Value=@n6)
  ]}
]
variables [
 CurrentPitch {type=number}
 CurrentRoll {type=number}
 DisabledTag {type=GameplayTag, default {tagName=Vehicle.Disabled}}
 HorizonMIDLeft {type=MaterialInstanceDynamic}
 HorizonMIDRight {type=MaterialInstanceDynamic}
 VehicleMaxAltitude {type=number}
]
widgets {
 info {inheritedWidgetCount=10, parentClass=/Script/HoagieUI.FortHoagieVehicleReticle, rootWidgetClass=&Overlay, widgetCount=47}
 widgets [
  Overlay_8 {widgetClassPath=&Overlay, widget=Overlay_8}
  CommonBorder_0 {
   widgetClassPath=&CommonBorder
   parent=Overlay_8
   widget=CommonBorder_0
   properties {contentColorAndOpacity=<rgba 1 1 1 0.7>, padding=<ltrb 0 0 0 0>}
   slot {class=OverlaySlot, properties {horizontalAlignment=HAlign_Center, verticalAlignment=VAlign_Center}}
  }
  HorizontalBox_0 {
   widgetClassPath=&HorizontalBox
   parent=CommonBorder_0
   widget=HorizontalBox_0
   slot {class=BorderSlot, properties {padding=<ltrb 0 0 0 0>}}
  }
  Overlay_LeftBracket {widgetClassPath=&Overlay, bIsVariable=true, parent=HorizontalBox_0, widget=Overlay_LeftBracket, slot {class=HorizontalBoxSlot}}
  Image_Speedometer {
   widgetClassPath=&Image
   bInherited=true
   bIsVariable=true
   parent=Overlay_LeftBracket
   widget=Image_Speedometer
   properties {
    visibility=HitTestInvisible
    brush {
     bIsDynamicallyLoaded=false
     drawAs=Image
     imageType=FullColor
     mirroring=NoMirror
     resourceName=None
     resourceObject=/TidalCraneHoagie/UI/Materials/MI_UI_TidalCraneHoagie_SpeedMeter
     tiling=NoTile
     imageSize=<460 460>
     margin=<ltrb 0 0 0 0>
     outlineSettings {
      bUseBrushTransparency=false
      roundingType=HalfHeightRadius
      width=0
      color {colorUseRule=UseColor_Specified, specifiedColor=<rgba 0 0 0 0>}
      cornerRadii=<0 0 0 0>
     }
     tintColor {colorUseRule=UseColor_Specified, specifiedColor=<rgba 1 1 1 1>}
     uVRegion {bIsValid=false, max=<0 0>, min=<0 0>}
    }
   }
   slot {class=OverlaySlot, properties {horizontalAlignment=HAlign_Center, verticalAlignment=VAlign_Center}}
  }
  Image_Speedometer_Glow {
   widgetClassPath=&Image
   bIsVariable=true
   parent=Overlay_LeftBracket
   widget=Image_Speedometer_Glow
   properties {
    visibility=HitTestInvisible
    brush {
     bIsDynamicallyLoaded=false
     drawAs=Image
     imageType=FullColor
     mirroring=NoMirror
     resourceName=None
     resourceObject=/TidalCraneHoagie/UI/Materials/MI_UI_TidalCraneHoagie_SpeedMeter_Glow
     tiling=NoTile
     imageSize=<460 460>
     margin=<ltrb 0 0 0 0>
     outlineSettings {
      bUseBrushTransparency=false
      roundingType=HalfHeightRadius
      width=0
      color {colorUseRule=UseColor_Specified, specifiedColor=<rgba 0 0 0 0>}
      cornerRadii=<0 0 0 0>
     }
     tintColor {colorUseRule=UseColor_Specified, specifiedColor=<rgba 1 1 1 1>}
     uVRegion {bIsValid=false, max=<0 0>, min=<0 0>}
    }
   }
   slot {class=OverlaySlot, properties {horizontalAlignment=HAlign_Center, verticalAlignment=VAlign_Center}}
  }
  Bracket_Left {
   widgetClassPath=&Image
   bIsVariable=true
   parent=Overlay_LeftBracket
   widget=Bracket_Left
   properties {brush {
     bIsDynamicallyLoaded=false
     drawAs=Image
     imageType=FullColor
     mirroring=NoMirror
     resourceName=None
     resourceObject=/TidalCraneHoagie/UI/Materials/MI_UI_TidalCraneHoagie_BracketLeft
     tiling=NoTile
     imageSize=<245 245>
     margin=<ltrb 0 0 0 0>
     outlineSettings {
      bUseBrushTransparency=false
      roundingType=HalfHeightRadius
      width=0
      color {colorUseRule=UseColor_Specified, specifiedColor=<rgba 0 0 0 0>}
      cornerRadii=<0 0 0 0>
     }
     tintColor {colorUseRule=UseColor_Specified, specifiedColor=<rgba 1 1 1 1>}
     uVRegion {bIsValid=false, max=<0 0>, min=<0 0>}
    }}
   slot {class=OverlaySlot, properties {horizontalAlignment=HAlign_Center, verticalAlignment=VAlign_Center, padding=<ltrb 69 0 0 0>}}
  }
  Overlay_Center {
   widgetClassPath=&Overlay
   parent=HorizontalBox_0
   widget=Overlay_Center
   slot {class=HorizontalBoxSlot, properties {padding=<ltrb 8 0 8 0>}}
  }
  Image_Center {
   widgetClassPath=&Image
   parent=Overlay_Center
   widget=Image_Center
   properties {
    visibility=HitTestInvisible
    brush {
     bIsDynamicallyLoaded=false
     drawAs=NoDrawType
     imageType=FullColor
     mirroring=NoMirror
     resourceName=None
     resourceObject=null
     tiling=NoTile
     imageSize=<100 100>
     margin=<ltrb 0 0 0 0>
     outlineSettings {
      bUseBrushTransparency=false
      roundingType=HalfHeightRadius
      width=0
      color {colorUseRule=UseColor_Specified, specifiedColor=<rgba 0 0 0 0>}
      cornerRadii=<0 0 0 0>
     }
     tintColor {colorUseRule=UseColor_Specified, specifiedColor=<rgba 1 1 1 1>}
     uVRegion {bIsValid=false, max=<0 0>, min=<0 0>}
    }
   }
   slot {class=OverlaySlot, properties {horizontalAlignment=HAlign_Center, verticalAlignment=VAlign_Center, padding=<ltrb 155 0 155 0>}}
  }
  Overlay_RightBracket {widgetClassPath=&Overlay, bIsVariable=true, parent=HorizontalBox_0, widget=Overlay_RightBracket, slot {class=HorizontalBoxSlot}}
  Bracket_Right {
   widgetClassPath=&Image
   bIsVariable=true
   parent=Overlay_RightBracket
   widget=Bracket_Right
   properties {
    visibility=SelfHitTestInvisible
    brush {
     bIsDynamicallyLoaded=false
     drawAs=Image
     imageType=FullColor
     mirroring=NoMirror
     resourceName=None
     resourceObject=/TidalCraneHoagie/UI/Materials/MI_UI_TidalCraneHoagie_BracketRight
     tiling=NoTile
     imageSize=<245 245>
     margin=<ltrb 0 0 0 0>
     outlineSettings {
      bUseBrushTransparency=false
      roundingType=HalfHeightRadius
      width=0
      color {colorUseRule=UseColor_Specified, specifiedColor=<rgba 0 0 0 0>}
      cornerRadii=<0 0 0 0>
     }
     tintColor {colorUseRule=UseColor_Specified, specifiedColor=<rgba 1 1 1 1>}
     uVRegion {bIsValid=false, max=<0 0>, min=<0 0>}
    }
    renderTransform {angle=0, scale=<-1 1>, shear=<0 0>, translation=<0 0>}
   }
   slot {class=OverlaySlot, properties {horizontalAlignment=HAlign_Center, verticalAlignment=VAlign_Center, padding=<ltrb 0 0 69 0>}}
  }
  CommonBorder_FuelMeterContainer {
   widgetClassPath=&CommonBorder
   bIsVariable=true
   parent=Overlay_RightBracket
   widget=CommonBorder_FuelMeterContainer
   properties {padding=<ltrb 0 0 0 0>}
   slot {class=OverlaySlot, properties {horizontalAlignment=HAlign_Fill, verticalAlignment=VAlign_Fill}}
  }
  Overlay_0 {
   widgetClassPath=&Overlay
   parent=CommonBorder_FuelMeterContainer
   widget=Overlay_0
   slot {class=BorderSlot, properties {padding=<ltrb 0 0 0 0>}}
  }
  Image_BoostMeter {
   widgetClassPath=&Image
   bInherited=true
   bIsVariable=true
   parent=Overlay_0
   widget=Image_BoostMeter
   properties {
    visibility=HitTestInvisible
    brush {
     bIsDynamicallyLoaded=false
     drawAs=Image
     imageType=FullColor
     mirroring=NoMirror
     resourceName=None
     resourceObject=/TidalCraneHoagie/UI/Materials/M_UI_TidalCraneHoagie_BoostMeter
     tiling=NoTile
     imageSize=<170 170>
     margin=<ltrb 0 0 0 0>
     outlineSettings {
      bUseBrushTransparency=false
      roundingType=HalfHeightRadius
      width=0
      color {colorUseRule=UseColor_Specified, specifiedColor=<rgba 0 0 0 0>}
      cornerRadii=<0 0 0 0>
     }
     tintColor {colorUseRule=UseColor_Specified, specifiedColor=<rgba 1 1 1 1>}
     uVRegion {bIsValid=false, max=<0 0>, min=<0 0>}
    }
   }
   slot {class=OverlaySlot, properties {horizontalAlignment=HAlign_Center, verticalAlignment=VAlign_Center, padding=<ltrb 0 0 30 77>}}
  }
  Image_FuelMeter {
   widgetClassPath=&Image
   bIsVariable=true
   parent=Overlay_RightBracket
   widget=Image_FuelMeter
   properties {
    visibility=HitTestInvisible
    brush {
     bIsDynamicallyLoaded=false
     drawAs=Image
     imageType=FullColor
     mirroring=NoMirror
     resourceName=None
     resourceObject=/TidalCraneHoagie/UI/Materials/M_UI_TidalCraneHoagie_FuelMeter
     tiling=NoTile
     imageSize=<460 460>
     margin=<ltrb 0 0 0 0>
     outlineSettings {
      bUseBrushTransparency=false
      roundingType=HalfHeightRadius
      width=0
      color {colorUseRule=UseColor_Specified, specifiedColor=<rgba 0 0 0 0>}
      cornerRadii=<0 0 0 0>
     }
     tintColor {colorUseRule=UseColor_Specified, specifiedColor=<rgba 1 1 1 1>}
     uVRegion {bIsValid=false, max=<0 0>, min=<0 0>}
    }
    renderTransform {angle=0, scale=<-1 1>, shear=<0 0>, translation=<0 0>}
   }
   slot {class=OverlaySlot, properties {horizontalAlignment=HAlign_Center, verticalAlignment=VAlign_Center}}
  }
  Image_FuelMeter_Glow {
   widgetClassPath=&Image
   bIsVariable=true
   parent=Overlay_RightBracket
   widget=Image_FuelMeter_Glow
   properties {
    visibility=HitTestInvisible
    brush {
     bIsDynamicallyLoaded=false
     drawAs=Image
     imageType=FullColor
     mirroring=NoMirror
     resourceName=None
     resourceObject=/TidalCraneHoagie/UI/Materials/MI_UI_TidalCraneHoagie_FuelMeter_Glow
     tiling=NoTile
     imageSize=<460 460>
     margin=<ltrb 0 0 0 0>
     outlineSettings {
      bUseBrushTransparency=false
      roundingType=HalfHeightRadius
      width=0
      color {colorUseRule=UseColor_Specified, specifiedColor=<rgba 0 0 0 0>}
      cornerRadii=<0 0 0 0>
     }
     tintColor {colorUseRule=UseColor_Specified, specifiedColor=<rgba 1 1 1 1>}
     uVRegion {bIsValid=false, max=<0 0>, min=<0 0>}
    }
    renderTransform {angle=0, scale=<-1 1>, shear=<0 0>, translation=<0 0>}
   }
   slot {class=OverlaySlot, properties {horizontalAlignment=HAlign_Center, verticalAlignment=VAlign_Center}}
  }
  ScaleBox_Fuel_Mobile {
   widgetClassPath=&ScaleBox
   bIsVariable=true
   parent=Overlay_RightBracket
   widget=ScaleBox_Fuel_Mobile
   properties {stretch=UserSpecified}
   slot {class=OverlaySlot, properties {horizontalAlignment=HAlign_Center, verticalAlignment=VAlign_Center, padding=<ltrb 52 0 0 0>}}
  }
  VerticalBox_Fuel {widgetClassPath=&VerticalBox, parent=ScaleBox_Fuel_Mobile, widget=VerticalBox_Fuel, slot {class=ScaleBoxSlot}}
  Image_FuelDecorTop {
   widgetClassPath=&Image
   parent=VerticalBox_Fuel
   widget=Image_FuelDecorTop
   properties {
    visibility=SelfHitTestInvisible
    brush {
     bIsDynamicallyLoaded=false
     drawAs=Image
     imageType=FullColor
     mirroring=NoMirror
     resourceName=None
     resourceObject=/TidalCraneHoagie/UI/Materials/M_UI_TidalCraneHoagie_FuelIconDecor
     tiling=NoTile
     imageSize=<14 28>
     margin=<ltrb 0 0 0 0>
     outlineSettings {
      bUseBrushTransparency=false
      roundingType=HalfHeightRadius
      width=0
      color {colorUseRule=UseColor_Specified, specifiedColor=<rgba 0 0 0 0>}
      cornerRadii=<0 0 0 0>
     }
     tintColor {colorUseRule=UseColor_Specified, specifiedColor=<rgba 1 1 1 1>}
     uVRegion {bIsValid=false, max=<0 0>, min=<0 0>}
    }
   }
   slot {class=VerticalBoxSlot, properties {horizontalAlignment=HAlign_Left, verticalAlignment=VAlign_Top}}
  }
  Overlay_2 {
   widgetClassPath=&Overlay
   parent=VerticalBox_Fuel
   widget=Overlay_2
   slot {class=VerticalBoxSlot, properties {horizontalAlignment=HAlign_Center, verticalAlignment=VAlign_Center, padding=<ltrb 0 4 0 4>}}
  }
  Image_FuelMeter_Icon {
   widgetClassPath=&Image
   bIsVariable=true
   parent=Overlay_2
   widget=Image_FuelMeter_Icon
   properties {
    visibility=HitTestInvisible
    brush {
     bIsDynamicallyLoaded=false
     drawAs=Image
     imageType=FullColor
     mirroring=NoMirror
     resourceName=None
     resourceObject=/TidalCraneHoagie/UI/Materials/M_UI_TidalCraneHoagie_FuelIcon
     tiling=NoTile
     imageSize=<16 16>
     margin=<ltrb 0 0 0 0>
     outlineSettings {
      bUseBrushTransparency=false
      roundingType=HalfHeightRadius
      width=0
      color {colorUseRule=UseColor_Specified, specifiedColor=<rgba 0 0 0 0>}
      cornerRadii=<0 0 0 0>
     }
     tintColor {colorUseRule=UseColor_Specified, specifiedColor=<rgba 1 1 1 1>}
     uVRegion {bIsValid=false, max=<0 0>, min=<0 0>}
    }
    renderTransform {angle=0, scale=<-1 1>, shear=<0 0>, translation=<0 0>}
   }
   slot {class=OverlaySlot}
  }
  Fuel_Low {
   widgetClassPath=&Image
   bIsVariable=true
   parent=Overlay_2
   widget=Fuel_Low
   properties {
    brush {
     bIsDynamicallyLoaded=false
     drawAs=Image
     imageType=NoImage
     mirroring=NoMirror
     resourceName=None
     resourceObject=/Game/Athena/HUD/LTM/Wax/Materials/M_UI_TextGlow_Add_Gradient_Orange_Nevada
     tiling=NoTile
     imageSize=<16 16>
     margin=<ltrb 0 0 0 0>
     outlineSettings {
      bUseBrushTransparency=false
      roundingType=HalfHeightRadius
      width=0
      color {colorUseRule=UseColor_Specified, specifiedColor=<rgba 0 0 0 0>}
      cornerRadii=<0 0 0 0>
     }
     tintColor {colorUseRule=UseColor_Specified, specifiedColor=<rgba 1 1 1 1>}
     uVRegion {bIsValid=false, max=<0 0>, min=<0 0>}
    }
    renderTransform {angle=0, scale=<3.2 3.2>, shear=<0 0>, translation=<0 0>}
   }
   slot {class=OverlaySlot, properties {horizontalAlignment=HAlign_Center, verticalAlignment=VAlign_Center}}
  }
  Image_FuelDecorBottom {
   widgetClassPath=&Image
   parent=VerticalBox_Fuel
   widget=Image_FuelDecorBottom
   properties {
    visibility=SelfHitTestInvisible
    brush {
     bIsDynamicallyLoaded=false
     drawAs=Image
     imageType=FullColor
     mirroring=NoMirror
     resourceName=None
     resourceObject=&M_UI_TidalCraneHoagie_FuelIconDecor
     tiling=NoTile
     imageSize=<14 28>
     margin=<ltrb 0 0 0 0>
     outlineSettings {
      bUseBrushTransparency=false
      roundingType=HalfHeightRadius
      width=0
      color {colorUseRule=UseColor_Specified, specifiedColor=<rgba 0 0 0 0>}
      cornerRadii=<0 0 0 0>
     }
     tintColor {colorUseRule=UseColor_Specified, specifiedColor=<rgba 1 1 1 1>}
     uVRegion {bIsValid=false, max=<0 0>, min=<0 0>}
    }
    renderTransform {angle=0, scale=<1 -1>, shear=<0 0>, translation=<0 0>}
   }
   slot {class=VerticalBoxSlot, properties {horizontalAlignment=HAlign_Left, verticalAlignment=VAlign_Top}}
  }
  Overlay_Fuel {
   widgetClassPath=&Overlay
   bIsVariable=true
   parent=Overlay_RightBracket
   widget=Overlay_Fuel
   properties {clipping=ClipToBounds, renderTransform {angle=0, scale=<1 1>, shear=<0 0>, translation=<300 0>}, renderTransformPivot=<0 0.5>}
   slot {class=OverlaySlot, properties {verticalAlignment=VAlign_Center}}
  }
  Switcher_Fuel {
   widgetClassPath=&CommonWidgetSwitcherLegacy
   bIsVariable=true
   parent=Overlay_Fuel
   widget=Switcher_Fuel
   properties {renderTransformPivot=<0 0.5>}
   slot {class=OverlaySlot, properties {verticalAlignment=VAlign_Fill}}
  }
  VerticalBox_LowFuel {
   widgetClassPath=&VerticalBox
   bIsVariable=true
   parent=Switcher_Fuel
   widget=VerticalBox_LowFuel
   properties {renderTransformPivot=<0 0.5>}
   slot {class=WidgetSwitcherSlot, properties {horizontalAlignment=HAlign_Left, verticalAlignment=VAlign_Center, padding=<ltrb 8 8 8 8>}}
  }
  Overlay_Warning {
   widgetClassPath=&Overlay
   bIsVariable=true
   parent=VerticalBox_LowFuel
   widget=Overlay_Warning
   slot {class=VerticalBoxSlot, properties {horizontalAlignment=HAlign_Left}}
  }
  CommonTextBlock_6 {
   widgetClassPath=&CommonTextBlock
   parent=Overlay_Warning
   widget=CommonTextBlock_6
   properties {
    lineHeightPercentage=0.85
    text=Warning
    textTransformPolicy=ToUpper
    colorAndOpacity {colorUseRule=UseColor_Specified, specifiedColor=<rgba 0 0 0 0.25>}
    font {
     bForceMonospaced=false
     bMaterialIsStencil=false
     fontMaterial=null
     fontObject=&BurbankSmall
     letterSpacing=100
     monospacedWidth=1
     size=12
     skewAmount=0
     typefaceFontName=Bold
     outlineSettings {
      bApplyOutlineToDropShadows=false
      bMiteredCorners=false
      bSeparateFillAlpha=false
      outlineMaterial=null
      outlineSize=2
      outlineColor=<rgba 0 0 0 0.1>
     }
    }
    margin=<ltrb 0 -4 0 0>
    shadowOffset=<0 0>
   }
   slot {class=OverlaySlot}
  }
  CommonTextBlock_Warning {
   widgetClassPath=&CommonTextBlock
   bIsVariable=true
   parent=Overlay_Warning
   widget=CommonTextBlock_Warning
   properties {
    lineHeightPercentage=0.85
    text=Warning
    textTransformPolicy=ToUpper
    colorAndOpacity {colorUseRule=UseColor_Specified, specifiedColor=<rgba 1 0.05 0 1>}
    font {
     bForceMonospaced=false
     bMaterialIsStencil=false
     fontMaterial=/TidalCraneHoagie/UI/Materials/M_UI_TidalCraneHoagie_Additive_White
     fontObject=&BurbankSmall
     letterSpacing=100
     monospacedWidth=1
     size=12
     skewAmount=0
     typefaceFontName=Bold
     outlineSettings {
      bApplyOutlineToDropShadows=false
      bMiteredCorners=false
      bSeparateFillAlpha=false
      outlineMaterial=null
      outlineSize=2
      outlineColor=<rgba 0 0 0 0>
     }
    }
    margin=<ltrb 0 -4 0 0>
    shadowOffset=<0 0>
   }
   slot {class=OverlaySlot}
  }
  warning_Glow {
   widgetClassPath=&Image
   parent=Overlay_Warning
   widget=warning_Glow
   properties {
    renderOpacity=0.25
    brush {
     bIsDynamicallyLoaded=false
     drawAs=Box
     imageType=NoImage
     mirroring=NoMirror
     resourceName=None
     resourceObject=&M_UI_TextGlow_Add_Gradient_Orange_Nevada
     tiling=NoTile
     imageSize=<24 24>
     margin=<ltrb 0.5 0.5 0.5 0.5>
     outlineSettings {
      bUseBrushTransparency=false
      roundingType=HalfHeightRadius
      width=0
      color {colorUseRule=UseColor_Specified, specifiedColor=<rgba 0 0 0 0>}
      cornerRadii=<0 0 0 0>
     }
     tintColor {colorUseRule=UseColor_Specified, specifiedColor=<rgba 1 1 1 1>}
     uVRegion {bIsValid=false, max=<0 0>, min=<0 0>}
    }
   }
   slot {class=OverlaySlot, properties {horizontalAlignment=HAlign_Fill, verticalAlignment=VAlign_Fill, padding=<ltrb -8 -6 -8 -6>}}
  }
  CommonTextBlock_LowFuel {
   widgetClassPath=&CommonTextBlock
   bIsVariable=true
   parent=VerticalBox_LowFuel
   widget=CommonTextBlock_LowFuel
   properties {
    lineHeightPercentage=0.8
    text=LOW FUEL
    textTransformPolicy=ToUpper
    wrapTextAt=300
    font {
     bForceMonospaced=false
     bMaterialIsStencil=false
     fontMaterial=null
     fontObject=&Burbank
     letterSpacing=0
     monospacedWidth=1
     size=18
     skewAmount=0
     typefaceFontName=BurbankRegular-Black
     outlineSettings {
      bApplyOutlineToDropShadows=false
      bMiteredCorners=false
      bSeparateFillAlpha=false
      outlineMaterial=null
      outlineSize=0
      outlineColor=<rgba 0 0 0 0>
     }
    }
    margin=<ltrb 0 -2 0 0>
    shadowOffset=<0 0>
   }
   slot {class=VerticalBoxSlot}
  }
  VerticalBox_NoFuel {
   widgetClassPath=&VerticalBox
   bIsVariable=true
   parent=Switcher_Fuel
   widget=VerticalBox_NoFuel
   properties {renderTransformPivot=<0 0.5>}
   slot {class=WidgetSwitcherSlot, properties {horizontalAlignment=HAlign_Left, verticalAlignment=VAlign_Center, padding=<ltrb 8 8 8 8>}}
  }
  Overlay_Warning_3 {
   widgetClassPath=&Overlay
   bIsVariable=true
   parent=VerticalBox_NoFuel
   widget=Overlay_Warning_3
   slot {class=VerticalBoxSlot, properties {horizontalAlignment=HAlign_Left}}
  }
  CommonTextBlock_7 {
   widgetClassPath=&CommonTextBlock
   parent=Overlay_Warning_3
   widget=CommonTextBlock_7
   properties {
    lineHeightPercentage=0.85
    text=Warning
    textTransformPolicy=ToUpper
    colorAndOpacity {colorUseRule=UseColor_Specified, specifiedColor=<rgba 0 0 0 0.25>}
    font {
     bForceMonospaced=false
     bMaterialIsStencil=false
     fontMaterial=null
     fontObject=&BurbankSmall
     letterSpacing=100
     monospacedWidth=1
     size=12
     skewAmount=0
     typefaceFontName=Bold
     outlineSettings {
      bApplyOutlineToDropShadows=false
      bMiteredCorners=false
      bSeparateFillAlpha=false
      outlineMaterial=null
      outlineSize=2
      outlineColor=<rgba 0 0 0 0.1>
     }
    }
    margin=<ltrb 0 -4 0 0>
    shadowOffset=<0 0>
   }
   slot {class=OverlaySlot}
  }
  CommonTextBlock_Warning_3 {
   widgetClassPath=&CommonTextBlock
   bIsVariable=true
   parent=Overlay_Warning_3
   widget=CommonTextBlock_Warning_3
   properties {
    lineHeightPercentage=0.85
    text=Warning
    textTransformPolicy=ToUpper
    colorAndOpacity {colorUseRule=UseColor_Specified, specifiedColor=<rgba 1 0.05 0 1>}
    font {
     bForceMonospaced=false
     bMaterialIsStencil=false
     fontMaterial=&M_UI_TidalCraneHoagie_Additive_White
     fontObject=&BurbankSmall
     letterSpacing=100
     monospacedWidth=1
     size=12
     skewAmount=0
     typefaceFontName=Bold
     outlineSettings {
      bApplyOutlineToDropShadows=false
      bMiteredCorners=false
      bSeparateFillAlpha=false
      outlineMaterial=null
      outlineSize=2
      outlineColor=<rgba 0 0 0 0>
     }
    }
    margin=<ltrb 0 -4 0 0>
    shadowOffset=<0 0>
   }
   slot {class=OverlaySlot}
  }
  warning_Glow_5 {
   widgetClassPath=&Image
   parent=Overlay_Warning_3
   widget=warning_Glow_5
   properties {
    renderOpacity=0.25
    brush {
     bIsDynamicallyLoaded=false
     drawAs=Box
     imageType=NoImage
     mirroring=NoMirror
     resourceName=None
     resourceObject=&M_UI_TextGlow_Add_Gradient_Orange_Nevada
     tiling=NoTile
     imageSize=<24 24>
     margin=<ltrb 0.5 0.5 0.5 0.5>
     outlineSettings {
      bUseBrushTransparency=false
      roundingType=HalfHeightRadius
      width=0
      color {colorUseRule=UseColor_Specified, specifiedColor=<rgba 0 0 0 0>}
      cornerRadii=<0 0 0 0>
     }
     tintColor {colorUseRule=UseColor_Specified, specifiedColor=<rgba 1 1 1 1>}
     uVRegion {bIsValid=false, max=<0 0>, min=<0 0>}
    }
   }
   slot {class=OverlaySlot, properties {horizontalAlignment=HAlign_Fill, verticalAlignment=VAlign_Fill, padding=<ltrb -8 -6 -8 -6>}}
  }
  CommonTextBlock_LowFuel_1 {
   widgetClassPath=&CommonTextBlock
   bIsVariable=true
   parent=VerticalBox_NoFuel
   widget=CommonTextBlock_LowFuel_1
   properties {
    lineHeightPercentage=0.8
    text=NO FUEL
    textTransformPolicy=ToUpper
    wrapTextAt=300
    font {
     bForceMonospaced=false
     bMaterialIsStencil=false
     fontMaterial=null
     fontObject=&Burbank
     letterSpacing=0
     monospacedWidth=1
     size=18
     skewAmount=0
     typefaceFontName=BurbankRegular-Black
     outlineSettings {
      bApplyOutlineToDropShadows=false
      bMiteredCorners=false
      bSeparateFillAlpha=false
      outlineMaterial=null
      outlineSize=0
      outlineColor=<rgba 0 0 0 0>
     }
    }
    margin=<ltrb 0 -2 0 0>
    shadowOffset=<0 0>
   }
   slot {class=VerticalBoxSlot}
  }
  text_pitchdebug {
   widgetClassPath=&CommonTextBlock
   bIsVariable=true
   parent=Overlay_8
   widget=text_pitchdebug
   properties {
    style=&TextStyle-Base_C
    text=Text Block
    visibility=Collapsed
    font {
     bForceMonospaced=false
     bMaterialIsStencil=false
     fontMaterial=null
     fontObject=&NotoSans
     letterSpacing=0
     monospacedWidth=1
     size=24
     skewAmount=0
     typefaceFontName=NotoSans-Regular
     outlineSettings {
      bApplyOutlineToDropShadows=false
      bMiteredCorners=false
      bSeparateFillAlpha=false
      outlineMaterial=null
      outlineSize=0
      outlineColor=<rgba 0 0 0 1>
     }
    }
    shadowOffset=<0 0>
   }
   slot {class=OverlaySlot, properties {horizontalAlignment=HAlign_Right}}
  }
  WBP_Vehicle_Warning {
   widgetClassPath=/Game/Athena/HUD/Vehicles/Dashboard/WBP_Vehicle_Warning
   bIsVariable=true
   parent=Overlay_8
   widget=WBP_Vehicle_Warning
   slot {class=OverlaySlot, properties {horizontalAlignment=HAlign_Center, verticalAlignment=VAlign_Center}}
  }
  Widget_FuelGauge {widgetClassPath=/Script/FortniteUI.FortVehicleFuelGauge, bInherited=true, widget=null}
  HorizontalBox_FuelCluster {widgetClassPath=&HorizontalBox, bInherited=true, widget=null}
  Image_BoostGlow {widgetClassPath=&Image, bInherited=true, widget=null}
  Overlay_BoostSparks {widgetClassPath=&Overlay, bInherited=true, widget=null}
  NumericText_Speed {widgetClassPath=/Script/CommonUI.CommonNumericTextBlock, bInherited=true, widget=null}
  Overlay_Gauge {widgetClassPath=&Overlay, bInherited=true, widget=null}
  Text_ItemName {widgetClassPath=&CommonTextBlock, bInherited=true, widget=null}
  TutorialHighlight {widgetClassPath=&TutorialHighlight}
 ]
}
